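# Sound cue for a capture point shutting down in the Disco limited-time mode, using dance-floor capture attenuation and the music emote sound class.
# A quality-level node selects between two variants.
# The first two levels mix the Domination crash-out music, slightly lowered and faded out over three seconds by an enveloper, with a wall-lights trap fire sound.
# The last two levels play only the unenveloped crash-out music.
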SoundCue /FNGameplayCues/Athena/Sounds/LTMs/Disco/LTM_Disco_CapturePoint_Shutdown_Cue
root=QualityLevel_0
nodes [
 {
  id=Enveloper_0
  children [WavePlayer_2]
  properties {pitchCurve {editorCurveData {keys [{time=0}]}}, volumeCurve {editorCurveData {keys [<at 0 1>,<at 3 0>]}}}
 }
 {id=Mixer_0, children [Enveloper_0,WavePlayer_0], properties {inputVolume [0.75,1]}}
 {id=QualityLevel_0, children [Mixer_0,Mixer_0,WavePlayer_1,WavePlayer_1]}
 {id=WavePlayer_0, properties {soundWaveAssetPtr=/FNGameplayCues/Sounds/Fort_Traps/Wall_Lights/WallTrap_Lights_Fire_004}}
 {id=WavePlayer_1, properties {soundWaveAssetPtr=&/Music_Domination_Crash_Out}}
 {id=WavePlayer_2, properties {soundWaveAssetPtr=&Music_Domination_Crash_Out}}
]
properties {
 attenuationSettings=&/LTM_Disco_DanceFloorCapture_Att
 maxDistance=11264
 soundClassObject=&Fort_Music_Emote
 volumeMultiplier=1
}
